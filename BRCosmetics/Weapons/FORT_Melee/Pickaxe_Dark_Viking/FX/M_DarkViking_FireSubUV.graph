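# Unlit translucent particle material for the Dark Viking pickaxe that renders SubUV flipbook fire recolored as ice flames.
# Inverted flipbook red samples an ice-flame curve atlas row, multiplied by particle color and passed through HDR emissive functions.
# Panning tiled noise, offset and scaled by dynamic parameters, breaks up flipbook alpha in a LerpNoise-controlled blend.
# Opacity combines this with particle alpha, near-camera fade and noise-biased depth fade.
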Material /BRCosmetics/Weapons/FORT_Melee/Pickaxe_Dark_Viking/FX/M_DarkViking_FireSubUV
graph {
 properties {
  bUsedWithNiagaraSprites=true
  bUsedWithParticleSprites=true
  blendMode=BLEND_Translucent
  pixelDepthOffsetMode=PDOM_Legacy
  shadingModel=MSM_Unlit
  usedShadingModels=Unlit
 }
 nodes [
  n1 = ScalarParameter() {defaultValue=200, parameterName=Near Camera Fade Distance}
  n2 = ParticleColor()
  n3 = DynamicParameter() {defaultValue=<rgba 0 0 0 1>, paramNames [LerpNoise, X, BlackBodyTemp, Scale]}
  n4 = Add(A=TextureCoordinate() {vTiling=0.5}, B=AppendVector(A=@n3.X, B=Add(A=@n3.X) {constB=2}))
  n5 = Floor(None=@n3.Scale)
  n6 = TextureSample(UVs=Panner(Coordinate=Multiply(A=@n4, B=@n5) {desc=Scale}) {speedY=-0.125}) {texture=/Game/Effects/Fort_Effects/Textures/Fire/T_FireTile}
  n7 = TextureSampleParameterSubUV!(UVs=TextureCoordinate(), Apply View MipBias) {parameterName=FireSubUV, texture=/Game/Effects/Fort_Effects/Textures/Fire/T_FireFluid_1_8X8}
  n8 = LinearInterpolate(A=Add(A=Multiply(A=@n6.A, B=@n7.A), B=@n7.A), B=Multiply(A=@n6.A, B=@n7.A), Alpha=@n3.LerpNoise)
  DepthFade(
   Opacity=Multiply(
    A=MaterialFunctionCall(Distance=@n1) {materialFunction=&Near_Camera_Fade}.Result
    B=Multiply(A=@n2.A, B=@n8)
   )
   FadeDistance=LinearInterpolate(Alpha=@n6.A) {constA=60, constB=120, desc=Bias Noise}
  )
  n9 = CurveAtlasRowParameter(CurveTime=OneMinus(None=@n7.R)) {
   atlas=&AthenaCurveAtlas01
   curve=/Game/Athena/LinearColorCurves/IceFlamesCurve
   defaultValue=5
   parameterName=Curve
  }
  n10 = Multiply(A=@n2.RGB, B=@n9)
  MaterialFunctionCall(Emissive=MaterialFunctionCall(LDR Input=Clamp(None=@n10), HDR Tint=@n10) {materialFunction=&DeriveHDRfromLDR}.RemappedRange) {materialFunction=&MF_FX_EmissiveAdjustments}
  Multiply(A=@n7.R, B=@n7.R)
 ]
}
