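# Additive unlit mesh-particle material for animated flames on the Dark Viking pickaxe, colored by default through an ice-flame curve.
# Three panning noise samples, offset by a dynamic parameter, combine with a vertical gradient to drive curve color, HDR emissive and opacity.
# The noise also displaces vertices along normals.
# A Pickaxe Mesh static switch selects tiling and inverted-Fresnel opacity, or an alternative with gradient opacity, non-mobile depth fade and extra displacement.
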Material /BRCosmetics/Weapons/FORT_Melee/Pickaxe_Dark_Viking/FX/M_DarkViking_MeshFlames
graph {
 properties {
  bUsedWithMeshParticles=true
  bUsedWithNiagaraMeshParticles=true
  blendMode=BLEND_Additive
  pixelDepthOffsetMode=PDOM_Legacy
  shadingModel=MSM_Unlit
  usedShadingModels=Unlit
 }
 nodes [
  n1 = ComponentMask(None=VectorParameter() {group=UVScale, parameterName=UVScale, defaultValue=<rgba 0.5 2 0 0>}.RGB) {g=true, r=true}
  n2 = Multiply(
   A=StaticSwitchParameter(True=TextureCoordinate() {uTiling=2, vTiling=0.5}, False=TextureCoordinate() {vTiling=0.5}) {defaultValue=true, parameterName=Pickaxe Mesh}
   B=@n1
  )
  n3 = DynamicParameter() {paramNames [UVOffset,Param2,Param3,Param4]}
  n4 = ScalarParameter() {defaultValue=-0.025, parameterName=X_Speed}
  n5 = ScalarParameter() {defaultValue=0.25, parameterName=Y_Speed}
  n6 = TextureSampleParameter2D(UVs=Panner(Coordinate=Add(A=@n2, B=@n3.UVOffset), Speed=AppendVector(A=@n4, B=@n5))) {parameterName=NoiseTexture, texture=&T_NoisePack_01}
  n7 = StaticComponentMaskParameter(None=AppendVector(A=@n6.RGB, B=@n6.RGB)) {defaultG=true, parameterName=Noise Mask}
  n8 = Constant() {r=0.2}
  n9 = Multiply(
   A=StaticSwitchParameter(True=TextureCoordinate() {uTiling=2}, False=TextureCoordinate() {uTiling=4}) {defaultValue=true, parameterName=Pickaxe Mesh}
   B=@n1
  )
  n10 = Multiply(A=Time(), B=ScalarParameter() {defaultValue=0.0625, parameterName=Speed})
  n11 = TextureSampleParameter2D(UVs=Panner(Coordinate=Add(A=@n9, B=@n3.UVOffset), Time=@n10) {speedX=0.1, speedY=1}) {parameterName=NoiseTexture, texture=&T_NoisePack_01}
  n12 = AppendVector(A=@n11.RGB, B=@n11.RGB)
  n13 = Multiply(
   A=@n1
   B=StaticSwitchParameter(True=TextureCoordinate() {uTiling=2}, False=TextureCoordinate() {uTiling=12, vTiling=3}) {defaultValue=true, parameterName=Pickaxe Mesh}
  )
  n14 = TextureSampleParameter2D(UVs=Panner(Coordinate=Add(A=@n3.UVOffset, B=@n13), Time=@n10) {speedX=-0.1, speedY=1}) {parameterName=NoiseTexture, texture=&T_NoisePack_01}
  n15 = Multiply(
   A=Add(
    A=MaterialFunctionCall(In=@n7, Contrast=@n8) {materialFunction=&CheapContrast}.Result
    B=StaticComponentMaskParameter(None=@n12) {defaultG=true, parameterName=Noise Mask}
   )
   B=StaticComponentMaskParameter(None=AppendVector(A=@n14.RGB, B=@n14.RGB)) {defaultG=true, parameterName=Noise Mask}
  )
  n16 = Add(
   A=@n15
   B=Power(
    Base=MaterialFunctionCall() {materialFunction=&LinearGradient}.VGradient
    Exp=ScalarParameter() {defaultValue=8, parameterName=FireLine}
   )
  )
  n17 = Power(Base=@n16) {constExponent=0.5}
  n18 = ScalarParameter() {defaultValue=0.5, parameterName=HardEdgeness}
  n19 = Clamp(None=StaticSwitchParameter(
    True=@n16
    False=MaterialFunctionCall(In=@n17, Contrast=@n18) {materialFunction=&CheapContrast}.Result
   ) {defaultValue=true, parameterName=Pickaxe Mesh})
  n20 = CurveAtlasRowParameter(CurveTime=@n19) {
   atlas=&AthenaCurveAtlas01
   curve=/Game/Athena/LinearColorCurves/IceFlamesCurve
   defaultValue=5
   parameterName=Curve
  }
  n21 = ParticleColor()
  n22 = Multiply(A=@n20, B=@n21.RGB)
  MaterialFunctionCall(Emissive=MaterialFunctionCall(LDR Input=Clamp(None=@n22), HDR Tint=@n22, HDR Intensity Multiplier=ScalarParameter() {defaultValue=1, parameterName=HDR_Intensity}) {materialFunction=&DeriveHDRfromLDR}.RemappedRange) {materialFunction=&MF_FX_EmissiveAdjustments}
  n23 = Fresnel(ExponentIn=ScalarParameter() {defaultValue=0.0156, parameterName=Frsnel Expo})
  n24 = Clamp(None=Multiply(A=OneMinus(None=@n23)) {constB=2})
  n25 = MaterialFunctionCall(In=Power(Base=@n16) {constExponent=0.5}, Contrast=ScalarParameter() {defaultValue=0.25, parameterName=OpacityEdgeness}) {materialFunction=&CheapContrast}
  n26 = Multiply(
   A=@n19
   B=Power(
    Base=MaterialFunctionCall() {materialFunction=&LinearGradient}.VGradient
    Exp=ScalarParameter() {defaultValue=8, parameterName=GradientContrast}
   )
  )
  n27 = DepthFade(Opacity=@n26, FadeDistance=ScalarParameter() {defaultValue=64, parameterName=FadeDistance}) {fadeDistanceDefault=50}
  Multiply(
   A=@n21.A
   B=StaticSwitchParameter(True=Multiply(A=@n24, B=@n25.Result), False=ShadingPathSwitch(Default=@n27, Mobile=@n26) {desc=Migrated from FeatureLevelSwitch}) {defaultValue=true, parameterName=Pickaxe Mesh}
  )
  n28 = ScalarParameter() {defaultValue=0.05, parameterName=Distort}
  n29 = Multiply(A=Multiply(A=VertexNormalWS(), B=@n15), B=@n28)
  n30 = TextureCoordinate()
  n31 = TextureSample(UVs=Panner(Coordinate=Add(A=@n3.UVOffset, B=@n30), Time=@n10) {speedY=0.25}) {texture=/Game/Global/Textures/Default/Masks_and_Gradients/CylGradient}
  StaticSwitchParameter(True=@n29, False=Add(A=@n29, B=Multiply(A=@n31.RGB, B=@n28) {constB=35})) {defaultValue=true, parameterName=Pickaxe Mesh}
 ]
}
