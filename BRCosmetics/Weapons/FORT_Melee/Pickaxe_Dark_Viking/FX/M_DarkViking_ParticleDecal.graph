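# Translucent particle material for the Dark Viking pickaxe that fakes a glowing, icy decal on nearby scene surfaces.
# Scene world normals blended with a water normal map drive a Fresnel term, shaped by a radial mask and colored through an ice-flame curve atlas row.
# Particle dynamic parameters control Fresnel exponent and emissive intensity.
# Nested depth fades confine opacity near geometry; mobile skips them and the scene normals.
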Material /BRCosmetics/Weapons/FORT_Melee/Pickaxe_Dark_Viking/FX/M_DarkViking_ParticleDecal
graph {
 properties {
  bTangentSpaceNormal=false
  bUsedWithNiagaraSprites=true
  bUsedWithParticleSprites=true
  blendMode=BLEND_Translucent
  pixelDepthOffsetMode=PDOM_Legacy
  translucencyLightingMode=TLM_Surface
  usedShadingModels=Default Lit
 }
 nodes [
  Constant() {r=1}
  Constant()
  n1 = MaterialFunctionCall(Normal=ComponentMask(None=SceneTexture() {sceneTextureId=PPI_WorldNormal}.Color) {b=true, g=true, r=true}, Flatness=Constant() {r=0.2}) {materialFunction=&FlattenNormal}
  n2 = Transform(None=TextureSample() {samplerType=SAMPLERTYPE_Normal, texture=/Game/Packages/Fortnite_UniqueMaterials_1/Textures/water2_NRM}.RGB)
  n3 = MaterialFunctionCall(
   BaseNormal=ShadingPathSwitch(Default=@n1.Result, Deferred=@n1.Result, Forward=Constant3Vector() {constant=<rgba 0 0 1 0>}, Mobile=@n1.Result)
   AdditionalNormal=@n2
  ) {materialFunction=&BlendAngleCorrectedNormals}
  n4 = TextureCoordinate()
  n5 = ScalarParameter() {defaultValue=0.5, parameterName=B/Radius}
  n6 = ScalarParameter() {parameterName=Hardness}
  n7 = Multiply(
   A=Fresnel(
    ExponentIn=DynamicParameter() {paramNames [FresnelExponent, EmissiveIntensity, Param3, Param4]}.FresnelExponent
    Normal=ShadingPathSwitch(Default=@n3.Result, Mobile=@n2) {desc=Migrated from FeatureLevelSwitch}
   ) {baseReflectFraction=0, exponent=1}
   B=Power(
    Base=SphereMask(A=@n4, B=@n5, Radius=@n5, Hardness=@n6) {hardnessPercent=0}
    Exp=ScalarParameter() {defaultValue=1.75, parameterName=ExpMask}
   )
  )
  n8 = CurveAtlasRowParameter(CurveTime=Power(Base=@n7) {constExponent=0.5}) {
   atlas=&AthenaCurveAtlas01
   curve=/Game/Athena/LinearColorCurves/IceFlamesCurve
   defaultValue=5
   parameterName=Curve
  }
  MaterialFunctionCall(Emissive=Multiply(A=@n8, B=DynamicParameter() {paramNames [FresnelExponent, EmissiveIntensity, Param3, Param4]}.EmissiveIntensity)) {materialFunction=&MF_FX_EmissiveAdjustments}
  n9 = ScalarParameter() {defaultValue=70, parameterName=VisibleDepthDistance}
  n10 = DepthFade(
   Opacity=OneMinus(None=DepthFade(FadeDistance=@n9) {fadeDistanceDefault=80})
   FadeDistance=ScalarParameter() {defaultValue=10, parameterName=DepthFade}
  ) {fadeDistanceDefault=15}
  Multiply(A=ParticleColor().A, B=ShadingPathSwitch(Default=Multiply(A=@n7, B=@n10), Mobile=@n7) {desc=Migrated from FeatureLevelSwitch})
 ]
}
