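# Translucent unlit, two-sided trail material for the Dark Viking pickaxe, built for ribbons and mesh particles.
# Distorted, panning flame, fire-mask, streak and dust layers drive an IceFlamesCurve atlas lookup tinted by particle color, with optional smoke-color modulation.
# Opacity combines those layers with UV masks, Fresnel, near-camera fade and switchable depth-fade edge blur.
# Switchable world position offset adds noise waves and a clamped bend driven by Niagara dynamic parameters.
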Material /BRCosmetics/Weapons/FORT_Melee/Pickaxe_Dark_Viking/FX/M_DarkViking_Trail
graph {
 properties {
  bEnableResponsiveAA=true
  bTangentSpaceNormal=false
  bUsedWithBeamTrails=true
  bUsedWithMeshParticles=true
  bUsedWithNiagaraMeshParticles=true
  bUsedWithNiagaraRibbons=true
  blendMode=BLEND_Translucent
  pixelDepthOffsetMode=PDOM_Legacy
  shadingModel=MSM_Unlit
  twoSided=true
  usedShadingModels=Unlit
 }
 nodes [
  Constant() {r=200}
  n1 = VectorParameter() {parameterName=Smoke Texture B color, defaultValue=<rgba 1 1 1 0>}
  n2 = TextureCoordinate()
  n3 = AppendVector(
   A=ScalarParameter() {defaultValue=2, parameterName=Smoke Texture B U}
   B=ScalarParameter() {defaultValue=0.5, parameterName=Smoke Texture B V}
  )
  n4 = TextureSampleParameter2D(UVs=Panner(Coordinate=Multiply(A=@n2, B=@n3)) {speedY=-0.5}) {parameterName=Smoke Texture B, texture=&T_NoisePack_01}
  n5 = Multiply(
   A=TextureCoordinate()
   B=AppendVector(
    A=ScalarParameter() {defaultValue=1, parameterName=Smoke Texture A U}
    B=ScalarParameter() {defaultValue=0.5, parameterName=Smoke Texture A V}
   )
  )
  n6 = TextureCoordinate()
  n7 = AppendVector(A=ScalarParameter() {defaultValue=2.5, parameterName=Turbulent U}, B=ScalarParameter() {defaultValue=1, parameterName=Turbulent V})
  n8 = TextureSample(UVs=Panner(Coordinate=Multiply(A=@n6, B=@n7)) {speedY=-0.3}) {samplerType=SAMPLERTYPE_LinearColor, texture=&T_NoisePack_LinearColor}
  n9 = Multiply(
   A=AppendVector(A=ComponentMask(None=@n8.RGB) {r=true}, B=ComponentMask(None=@n8.RGB) {g=true})
   B=ScalarParameter() {defaultValue=0.2, parameterName=Turbulent intensity}
  ) {constB=0.2}
  n10 = TextureSampleParameter2D(UVs=Panner(Coordinate=Add(A=@n5, B=@n9)) {speedY=-0.5}) {parameterName=Smoke Texture A, samplerType=SAMPLERTYPE_LinearColor, texture=&T_NoisePack_LinearColor}
  n11 = Multiply(A=@n10.RGB, B=@n10.RGB)
  n12 = VectorParameter() {parameterName=Smoke Texture A color, defaultValue=<rgba 1 1 1 0>}
  n13 = Add(
   A=Desaturation(None=Multiply(A=@n1.RGB, B=@n4.B), Fraction=ScalarParameter() {parameterName=Smoke Texture B Desaturation})
   B=Desaturation(None=Multiply(A=@n11, B=@n12.RGB), Fraction=ScalarParameter() {parameterName=Smoke Texture A Desaturation})
  )
  n14 = TextureSample(UVs=Panner(Coordinate=TextureCoordinate() {uTiling=0.5, vTiling=0.25}) {bFractionalPart=true, speedY=-0.65}) {samplerSource=SSM_Wrap_WorldGroupSettings, texture=/Game/Global/Textures/Default/PointSplatter/T_Tiling_Dust}
  n15 = ScalarParameter() {defaultValue=1.5, parameterName=UScale}
  n16 = ScalarParameter() {defaultValue=1, parameterName=VScale}
  n17 = TextureSample(UVs=Panner(Coordinate=TextureCoordinate() {coordinateIndex=1}) {speedX=1, speedY=0.1}) {
   samplerSource=SSM_Wrap_WorldGroupSettings
   samplerType=SAMPLERTYPE_Normal
   texture=/Game/Effects/Fort_Effects/Gameplay/SmokeDustExplosions/Textures/T_Explosion_NM_01
  }
  n18 = ComponentMask(None=MaterialFunctionCall(Normal=@n17.RGB, Flatness=DynamicParameter() {defaultValue=<rgba 0.975 1 1 1>, paramNames [UVDistortion,Param2,Param3,Param4]}.UVDistortion) {materialFunction=&FlattenNormal}.Result) {g=true, r=true}
  n19 = Add(A=Multiply(A=TextureCoordinate() {coordinateIndex=1, uTiling=2}, B=AppendVector(A=@n15, B=@n16)), B=@n18)
  n20 = Add(A=@n19) {constB=0.37}
  n21 = ScalarParameter() {defaultValue=2, parameterName=Bias}
  n22 = TextureSample!(
   UVs=StaticSwitchParameter(True=Panner(Coordinate=@n20) {speedY=-0.8}, False=Panner(Coordinate=@n20) {speedX=-0.8, speedY=-0.8}) {parameterName=Use Custom Panner}
   Tex
   Bias=@n21
   Apply View MipBias
  ) {
   mipValueMode=TMVM_MipBias
   samplerSource=SSM_Wrap_WorldGroupSettings
   texture=/Game/Effects/Fort_Effects/Textures/Tile/T_Tile_Texture_Cartoon_Flames_Layer
  }
  n23 = TextureSample(
   UVs=StaticSwitchParameter(True=Panner(Coordinate=@n19) {speedY=-0.6}, False=Panner(Coordinate=@n19) {speedX=-0.6, speedY=-0.6}) {parameterName=Use Custom Panner}
   Bias=@n21
  ) {
   mipValueMode=TMVM_MipBias
   samplerSource=SSM_Wrap_WorldGroupSettings
   texture=&T_Tile_Texture_Cartoon_Flames_Layer
  }
  n24 = Add(A=Multiply(A=@n22.R) {constB=0.5}, B=Multiply(A=@n23.R) {constB=0.5})
  n25 = TextureCoordinate()
  n26 = Power(Base=OneMinus(None=ComponentMask(None=@n25) {g=true})) {constExponent=10}
  n27 = Add(A=TextureCoordinate() {coordinateIndex=1}, B=@n18)
  n28 = OneMinus(None=ComponentMask(None=@n27) {g=true})
  n29 = OneMinus(None=ComponentMask(None=@n27) {g=true})
  n30 = TextureCoordinate()
  n31 = LinearInterpolate(
   A=AppendVector(A=ComponentMask(None=@n27) {r=true}, B=Add(A=@n28) {constB=-0.15})
   B=AppendVector(A=ComponentMask(None=@n27) {r=true}, B=Add(A=@n29) {constB=0.45})
   Alpha=Sine(None=ComponentMask(None=@n30) {r=true}) {period=2}
  )
  n32 = ScalarParameter() {defaultValue=1.5, parameterName=Bias_A}
  n33 = TextureSample(
   UVs=StaticSwitchParameter(True=Panner(Coordinate=@n31) {speedX=0.5}, False=Panner(Coordinate=@n31) {speedX=-1}) {parameterName=Use Custom Panner}
   Bias=@n32
  ) {mipValueMode=TMVM_MipBias, texture=/Game/Effects/Fort_Effects/Textures/Fire/T_Fire_Mask}
  n34 = Add(A=TextureCoordinate() {coordinateIndex=1}, B=@n18)
  n35 = OneMinus(None=ComponentMask(None=@n34) {g=true})
  n36 = OneMinus(None=ComponentMask(None=@n34) {g=true})
  n37 = TextureCoordinate()
  n38 = LinearInterpolate(
   A=AppendVector(A=ComponentMask(None=@n34) {r=true}, B=Add(A=@n35) {constB=-0.15})
   B=AppendVector(A=ComponentMask(None=@n34) {r=true}, B=Add(A=@n36) {constB=0.45})
   Alpha=Sine(None=ComponentMask(None=@n37) {r=true}) {period=2}
  )
  n39 = TextureSample(
   UVs=StaticSwitchParameter(True=Panner(Coordinate=@n38) {speedX=-0.5}, False=Panner(Coordinate=@n38) {speedX=-1}) {parameterName=Use Custom Panner}
   Bias=@n32
  ) {mipValueMode=TMVM_MipBias, texture=&T_Fire_Mask}
  n40 = Clamp(None=Multiply(A=Add(A=@n24, B=@n26), B=Multiply(A=@n33.R, B=@n39.R)))
  n41 = TextureCoordinate() {uTiling=3.464}
  n42 = Power(Base=TextureSample!(UVs=Panner(Coordinate=@n41) {speedX=0.976}, Tex, Apply View MipBias) {samplerSource=SSM_Wrap_WorldGroupSettings, texture=/Game/Effects/Fort_Effects/Textures/Outlander/T_InTheZone_LightRays_01}.R) {constExponent=1.5}
  n43 = TextureSample(UVs=Panner(Coordinate=TextureCoordinate() {uTiling=4.4362}) {speedX=-0.7616}) {samplerSource=SSM_Wrap_WorldGroupSettings, texture=&T_InTheZone_LightRays_01}
  n44 = Multiply(A=@n43.G, B=@n43.G)
  n45 = LinearInterpolate(
   A=ScalarParameter() {parameterName=Fine Streaks Opacity A}
   B=ScalarParameter() {defaultValue=0.2, parameterName=Fine Streaks Opacity B}
   Alpha=Multiply(A=Add(A=@n42, B=@n44)) {constB=0.5}
  ) {constA=1}
  n46 = ScalarParameter() {defaultValue=0.8, parameterName=Hard Streaks Opacity}
  n47 = TextureSample(UVs=Panner(Coordinate=TextureCoordinate() {vTiling=0.2}) {bFractionalPart=true, speedY=-0.65}) {samplerSource=SSM_Wrap_WorldGroupSettings, texture=&T_NoisePack_01}
  n48 = ComponentMask(None=TextureCoordinate()) {g=true}
  n49 = Power(Base=Add(A=ConstantBiasScale(None=@n47.R) {bias=-0.5, scale=0.2}, B=OneMinus(None=@n48))) {constExponent=4}
  n50 = Add(A=@n40, B=Clamp(None=LinearInterpolate(A=@n45, B=@n46, Alpha=@n49) {constAlpha=0}))
  n51 = CurveAtlasRowParameter(CurveTime=Clamp(None=Add(A=@n14.R, B=@n50))) {
   atlas=&AthenaCurveAtlas01
   curve=/Game/Athena/LinearColorCurves/IceFlamesCurve
   defaultValue=5
   parameterName=Cuve
  }
  n52 = CurveAtlasRowParameter(CurveTime=@n40) {
   atlas=&AthenaCurveAtlas01
   curve=&IceFlamesCurve
   defaultValue=5
   parameterName=Cuve
  }
  n53 = Multiply(A=@n52) {constB=1.5}
  n54 = OneMinus(None=ComponentMask(None=TextureCoordinate()) {r=true})
  n55 = Add(A=@n40, B=@n14.R)
  n56 = ParticleColor()
  n57 = Multiply(A=LinearInterpolate(A=@n51, B=LinearInterpolate(A=@n52, B=@n53, Alpha=@n54) {constA=1, constB=10}, Alpha=Clamp(None=@n55)), B=@n56.RGB)
  n58 = Multiply(A=@n13, B=@n57)
  MaterialFunctionCall(Emissive=StaticSwitchParameter(True=ParticleColor().RGB, False=StaticSwitchParameter(True=@n58, False=@n57) {parameterName=Use Custom Color}) {parameterName=Only Use particle color}) {materialFunction=&MF_FX_EmissiveAdjustments}
  n59 = ComponentMask(None=TextureCoordinate()) {g=true}
  OneMinus(None=@n59)
  n60 = TextureCoordinate()
  OneMinus(None=@n60)
  n61 = MaterialFunctionCall(Float2=TextureCoordinate()) {materialFunction=&BreakOutFloat2Components}
  n62 = Multiply(A=@n61.R) {constB=3}
  n63 = Multiply(A=Multiply(A=Saturate(None=@n62), B=OneMinus(None=@n61.R)), B=OneMinus(None=@n61.G))
  n64 = OneMinus(None=SphereMask(A=TextureCoordinate() {uTiling=0.5}, B=Constant2Vector() {g=0.95, r=0.85}, Radius=Constant() {r=0.7}, Hardness=Constant() {r=0.3}))
  n65 = Multiply(A=@n64, B=@n64)
  n66 = Multiply(A=@n64, B=Multiply(A=@n65, B=@n65))
  n67 = Multiply(A=Multiply(A=Multiply(A=@n50, B=@n63), B=Clamp(None=@n66))) {constB=3}
  n68 = ScalarParameter() {defaultValue=64, parameterName=NearFadeDistance}
  n69 = Multiply(
   A=Multiply(
    A=Saturate(None=@n67)
    B=MaterialFunctionCall(Distance=@n68) {materialFunction=&Near_Camera_Fade}.Result
   )
   B=MaterialFunctionCall(
    Brightness=ScalarParameter() {defaultValue=1.5, parameterName=Fresnel Brightness}
    Exponent=ScalarParameter() {defaultValue=3, parameterName=Fresnel Exponent}
   ) {materialFunction=&MF_Fresnel_FX}.Fresnel Double Sided
  )
  n70 = ComponentMask(None=TextureCoordinate()) {g=true}
  n71 = ScalarParameter() {defaultValue=2, parameterName=Mask Power}
  n72 = DepthFade(Opacity=Multiply(A=@n69, B=Power(Base=@n70, Exp=@n71)), FadeDistance=ScalarParameter() {defaultValue=50, parameterName=Depth Fade}) {fadeDistanceDefault=50}
  n73 = ScalarParameter() {defaultValue=1, parameterName=Alpha}
  n74 = Multiply(A=@n56.A, B=StaticSwitchParameter(True=Multiply(A=@n72, B=@n73), False=@n69) {parameterName=Edge blur})
  StaticSwitchParameter(True=Multiply(A=@n13, B=@n74), False=@n74) {parameterName=Use Custom Color}
  n75 = Constant()
  n76 = DynamicParameter() {desc=Parameters set by Niagara, parameterIndex=1, paramNames [X,Y,Z,WPOIntensity]}
  n77 = MaterialFunctionCall(X=@n76.X, Y=@n76.X, Z=@n76.X) {materialFunction=&MakeFloat3}
  n78 = ParticlePositionWS()
  n79 = ComponentMask(None=@n60) {g=true}
  n80 = ScalarParameter() {defaultValue=3, parameterName=WPOBendPower}
  n81 = LinearInterpolate(
   B=Clamp(
    None=Subtract(A=@n77.Result, B=@n78)
    Min=ScalarParameter() {defaultValue=-30, parameterName=Clamp_min}
    Max=ScalarParameter() {defaultValue=30, parameterName=Clamp_max}
   ) {maxDefault=1000, minDefault=-1000}
   Alpha=Multiply(A=Power(Base=@n79, Exp=@n80), B=@n76.WPOIntensity)
  )
  n82 = ScalarParameter() {defaultValue=1, parameterName=Use interpolate WPO}
  n83 = DynamicParameter() {parameterIndex=3, paramNames [Offset, Texture_mask_offset, WaveIntensity, Param4]}
  n84 = Panner(
   Coordinate=Add(A=@n83.Offset, B=TextureCoordinate())
   Time=Multiply(A=Time(), B=ScalarParameter() {defaultValue=1, group=Panner Speed, parameterName=Speed})
  ) {speedY=0.8}
  n85 = AppendVector(
   A=ComponentMask(None=TextureSample(UVs=@n84) {samplerType=SAMPLERTYPE_Normal, texture=&T_Weap_Distort_N_02}.RGB) {g=true, r=true}
   B=TextureSample(UVs=@n84) {texture=&T_Smooth_tile}.R
  )
  n86 = TextureCoordinate()
  n87 = Add(A=Multiply(A=ComponentMask(None=@n86) {g=true}, B=@n83.Param4), B=Time())
  n88 = Add(A=Multiply(A=Sine(None=@n87), B=ScalarParameter() {defaultValue=0.19, parameterName=Param_3}), B=@n85)
  n89 = ScalarParameter() {parameterName=Add_Noise}
  n90 = Multiply(
   A=LinearInterpolate(B=LinearInterpolate(A=@n85, B=@n88, Alpha=@n89), Alpha=@n59)
   B=DynamicParameter() {parameterIndex=2, paramNames [Offset, Texture_mask_offset, WaveIntensity, Param4]}.WaveIntensity
  ) {constB=10}
  n91 = ScalarParameter() {desc=no_Wpo, group=Simple_Round_particle, parameterName=No_WPO}
  StaticSwitchParameter(True=Add(A=LinearInterpolate(A=@n75, B=@n81, Alpha=@n82), B=LinearInterpolate(A=@n90, Alpha=@n91)), False=Constant()) {parameterName=WPO}
 ]
}
